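# Sound cue for the CubeQueenFemale pickaxe's impact on an enemy, playing randomized player-hit waves.
# A quality-level root sends its first two slots to a six-wave random pool and its last two to a three-wave pool.
# Each pool passes through a Modulator for variation and is excluded from branch culling.
# The cue primes on load and uses pickaxe-impact attenuation, the impacts sound class and melee-weapon concurrency.
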SoundCue /BRCosmetics/Sounds/PickAxes/CubeQueenFemale/PickAxe_CubeQueenFemale_Enemy_Impact_Cue
root=QualityLevel_0
nodes [
 {id=Modulator_0, children [Random_2]}
 {id=Modulator_1, children [Random_3]}
 {id=QualityLevel_0, children [Modulator_0,Modulator_0,Modulator_1,Modulator_1]}
 {
  id=Random_2
  children [WavePlayer_9, WavePlayer_10, WavePlayer_11, WavePlayer_12, WavePlayer_13, WavePlayer_14]
  properties {bShouldExcludeFromBranchCulling=true, weights [1,1,1,1,1,1]}
 }
 {id=Random_3, children [WavePlayer_15, WavePlayer_16, WavePlayer_17], properties {bShouldExcludeFromBranchCulling=true, weights [1,1,1]}}
 {id=WavePlayer_10, properties {soundWaveAssetPtr=&/Pickaxe_CubeQueenFemale_Hit_Player_02}}
 {id=WavePlayer_11, properties {soundWaveAssetPtr=&/Pickaxe_CubeQueenFemale_Hit_Player_03}}
 {id=WavePlayer_12, properties {soundWaveAssetPtr=&/Pickaxe_CubeQueenFemale_Hit_Player_04}}
 {id=WavePlayer_13, properties {soundWaveAssetPtr=&/Pickaxe_CubeQueenFemale_Hit_Player_05}}
 {id=WavePlayer_14, properties {soundWaveAssetPtr=&/Pickaxe_CubeQueenFemale_Hit_Player_06}}
 {id=WavePlayer_15, properties {soundWaveAssetPtr=&/Pickaxe_CubeQueenFemale_Hit_Player_01}}
 {id=WavePlayer_16, properties {soundWaveAssetPtr=&Pickaxe_CubeQueenFemale_Hit_Player_02}}
 {id=WavePlayer_17, properties {soundWaveAssetPtr=&Pickaxe_CubeQueenFemale_Hit_Player_03}}
 {id=WavePlayer_9, properties {soundWaveAssetPtr=&Pickaxe_CubeQueenFemale_Hit_Player_01}}
]
properties {
 attenuationSettings=/Game/Sounds/Attenuation/Fort_Pickaxe_Impact_Att
 audioPropertiesSheet=/BRCosmetics/Sounds/Settings/PropertySheets/Pickaxe/APS_Cosmetic_Pickaxe_Impact
 bPrimeOnLoad=true
 maxDistance=4300
 soundClassObject=&Fort_Impacts
 volumeMultiplier=0.675
 attenuationOverrides {
  bAttenuateWithLPF=true
  distanceAlgorithm=Logarithmic
  falloffDistance=800
  lPFRadiusMax=800
  lPFRadiusMin=400
  attenuationShapeExtents {x=300}
 }
 concurrencyOverrides {bLimitToOwner=true, maxCount=2, resolutionRule=StopOldest}
 concurrencySet [&Player_Melee_Weapon_Conc]
}
