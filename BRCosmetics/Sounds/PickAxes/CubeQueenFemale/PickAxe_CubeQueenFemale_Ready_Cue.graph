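# Sound cue for the ready sound of the CubeQueenFemale cosmetic pickaxe, built from swing recordings.
# A mixer blends a close-attenuated layer of pitched-down CubeQueen swings, at a higher input level, with a medium-attenuated layer of generic distant fire axe swings.
# In each layer, a quality-level node picks between random pools of six or three waves to trim variation on some quality levels.
# Owner-limited concurrency stops the oldest instance.
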SoundCue /BRCosmetics/Sounds/PickAxes/CubeQueenFemale/PickAxe_CubeQueenFemale_Ready_Cue
root=Mixer_2
nodes [
 {id=Attenuation_2, children [Modulator_0], properties {attenuationSettings=/Game/Sounds/Attenuation/Fort_Weapon_Pickaxe_Close_Att}}
 {id=Attenuation_3, children [QualityLevel_0], properties {attenuationSettings=/Game/Sounds/Attenuation/Fort_Weapon_Pickaxe_Medium_Att}}
 {id=Mixer_2, children [Attenuation_2, Attenuation_3], properties {inputVolume [0.7,1]}}
 {id=Modulator_0, children [QualityLevel_1], properties {pitchMax=0.8, pitchMin=0.8, volumeMax=1.2}}
 {id=QualityLevel_0, children [Random_2,Random_2,Random_3,Random_3]}
 {id=QualityLevel_1, children [Random_0,Random_0,Random_1,Random_1]}
 {
  id=Random_0
  children [WavePlayer_2,WavePlayer_3,WavePlayer_4,WavePlayer_5,WavePlayer_6,WavePlayer_7]
  properties {bShouldExcludeFromBranchCulling=true, weights [1,1,1,1,1,1]}
 }
 {id=Random_1, children [WavePlayer_8, WavePlayer_9, WavePlayer_10], properties {bShouldExcludeFromBranchCulling=true, weights [1,1,1]}}
 {
  id=Random_2
  children [WavePlayer_0, WavePlayer_36, WavePlayer_37, WavePlayer_38, WavePlayer_39, WavePlayer_40]
  properties {weights [1,1,1,1,1,1]}
 }
 {id=Random_3, children [WavePlayer_42, WavePlayer_41, WavePlayer_43], properties {bShouldExcludeFromBranchCulling=true, weights [1,1,1]}}
 {id=WavePlayer_0, properties {soundWaveAssetPtr=/Game/Athena/Sounds/Weapons/Fire_Axe_Swing_Distant_04}}
 {id=WavePlayer_10, properties {soundWaveAssetPtr=&/Pickaxe_CubeQueenFemale_Swing_03}}
 {id=WavePlayer_2, properties {soundWaveAssetPtr=&/Pickaxe_CubeQueenFemale_Swing_01}}
 {id=WavePlayer_3, properties {soundWaveAssetPtr=&/Pickaxe_CubeQueenFemale_Swing_02}}
 {id=WavePlayer_36, properties {soundWaveAssetPtr=/Game/Athena/Sounds/Weapons/Fire_Axe_Swing_Distant_01}}
 {id=WavePlayer_37, properties {soundWaveAssetPtr=/Game/Athena/Sounds/Weapons/Fire_Axe_Swing_Distant_02}}
 {id=WavePlayer_38, properties {soundWaveAssetPtr=/Game/Athena/Sounds/Weapons/Fire_Axe_Swing_Distant_03}}
 {id=WavePlayer_39, properties {soundWaveAssetPtr=/Game/Athena/Sounds/Weapons/Fire_Axe_Swing_Distant_05}}
 {id=WavePlayer_4, properties {soundWaveAssetPtr=&Pickaxe_CubeQueenFemale_Swing_03}}
 {id=WavePlayer_40, properties {soundWaveAssetPtr=/Game/Athena/Sounds/Weapons/Fire_Axe_Swing_Distant_06}}
 {id=WavePlayer_41, properties {soundWaveAssetPtr=&Fire_Axe_Swing_Distant_03}}
 {id=WavePlayer_42, properties {soundWaveAssetPtr=&Fire_Axe_Swing_Distant_04}}
 {id=WavePlayer_43, properties {soundWaveAssetPtr=&Fire_Axe_Swing_Distant_06}}
 {id=WavePlayer_5, properties {soundWaveAssetPtr=&/Pickaxe_CubeQueenFemale_Swing_04}}
 {id=WavePlayer_6, properties {soundWaveAssetPtr=&/Pickaxe_CubeQueenFemale_Swing_05}}
 {id=WavePlayer_7, properties {soundWaveAssetPtr=&/Pickaxe_CubeQueenFemale_Swing_06}}
 {id=WavePlayer_8, properties {soundWaveAssetPtr=&Pickaxe_CubeQueenFemale_Swing_01}}
 {id=WavePlayer_9, properties {soundWaveAssetPtr=&Pickaxe_CubeQueenFemale_Swing_02}}
]
properties {
 audioPropertiesSheet=/BRCosmetics/Sounds/Settings/PropertySheets/Pickaxe/APS_Cosmetic_Pickaxe_Ready
 bOverrideConcurrency=true
 bPrimeOnLoad=true
 maxDistance=3000
 soundClassObject=&Fort_WeaponMelee_PSM
 volumeMultiplier=0.2
 concurrencyOverrides {bLimitToOwner=true, maxCount=2, resolutionRule=StopOldest}
 concurrencySet [&Player_Melee_Weapon_Conc]
}
